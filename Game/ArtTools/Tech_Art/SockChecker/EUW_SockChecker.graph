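# Editor utility widget that checks a character item definition's body part for sock morph adjustment setup.
# A change in the embedded cosmetic picker or the details view runs Python to find the body part.
# Blueprint then validates its morph adjustment asset and material parameters.
# On success, a second Python script reports which variants have CosmeticProperties, shown in a green message panel; missing data shows red.
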EditorUtilityWidgetBlueprint /Game/ArtTools/Tech_Art/SockChecker/EUW_SockChecker
graphs [
 EventGraph {nodes [
   n1 = "On Property Changed (DetailsView_56)"!() -> (OutputDelegate, then, PropertyName:Name=None) {is=title K2Node_ComponentBoundEvent}
   n2 = Execute Python Script!(
    execute=[@n1.then, @n13.then]
    PythonScript="import unreal\r\n\r\n##setup error returns\r\ntoolfail = False\r\ncp_property_present = False\r\nmorph_materials_parameters_present = False\r\nreturnstruct = None\r\n\r\nidParts = cid.get_editor_property(\"BaseCharacterParts\")\r\n\r\nbodyCP = -1.0\r\n\r\n##find body\r\ndef findBodypart(idParts):\r\n\tfor entry in idParts:\r\n\t\tprint(entry.get_name().lower())\r\n\t\tif \"body\" in entry.get_name().lower():\r\n\t\t\tbodyCP = entry\r\n\treturn bodyCP\r\n\r\n\tif len(bodyCP.get_editor_property(\"CosmeticPartDataList\")) > 0:\r\n\t\treturn bodyCP.get_editor_property(\"CosmeticPartDataList\")[0]\r\n\r\nbodypart = findBodypart(idParts)"
    cid:Athena Character Item Definition Object Reference=Get SockCharacter() -> (SockCharacter:Athena Character Item Definition Object Reference)
   ) -> (then, ReturnValue, bodypart:Custom Character Part Object Reference) {inputs [cid], outputs [bodypart]}
   > n3 = Set BodyCP(BodyCP:Custom Character Part Object Reference=@n2.bodypart) -> (Output_Get:Custom Character Part Object Reference)
   > n4 = GetCosmeticPartData!(execute, self:Custom Character Part Object Reference=@n3) -> (Found:Exec, NotFound:Exec, CosmeticPartData:Cosmetic Part Morph Adjustment Data Structure)
   > n5 = Set Morph!(execute, Morph:Cosmetic Part Morph Adjustment Data Structure=@n4) -> (then, Output_Get_MorphAdjustmentDataAsset:Cosmetic Morph Adjustment Data Soft Object Reference)
   > n6 = LoadAsset_Blocking(Asset=@n5)
   > n7 = Cast To CosmeticMorphAdjustmentData!(execute, Object=@n6) -> (then, CastFailed, AsCosmetic Morph Adjustment Data:Cosmetic Morph Adjustment Data Object Reference) {is=title K2Node_DynamicCast}
   > n8 = Set MorphMaterial(MorphMaterial:Cosmetic Morph Material Params Object Reference=Get MorphMaterialParameters!(self:Cosmetic Morph Adjustment Data Object Reference=@n7) -> (MorphMaterialParameters:Cosmetic Morph Material Params Object Reference)) -> (Output_Get:Cosmetic Morph Material Params Object Reference)
   > n9 = Is Valid(InputObject=@n8)
   Update Message!(
    execute=@n9.Is Not Valid
    self
    Message:String=CosmeticMorphAdjustmentData Missing Parameter asset
    MessageColor:Linear Color Structure=(R=1,G=0,B=0,A=1)
   ) -> (then)
   n10 = GatherVariantsAndMorph!(execute=@n9.Is Valid, self, SockCharacter:Athena Character Item Definition Object Reference=Get SockCharacter()) -> (then, MorphType:String)
   > Update Message(Message=Append(A="Asset Clear to check in viewport, has the following variants:\r\n", B=@n10), MessageColor=(R=0,G=0.2812,B=0.0053,A=1))
   Update Message(execute=[@n7.CastFailed, @n4.NotFound], Message=Body Missing Morph Adjustment Data asset, MessageColor=(R=1,G=0,B=0,A=1))
   n11 = SetInternal!() -> (OutputDelegate, then, PropertyName:Name) {is=title K2Node_CustomEvent}
   > Set Morph!(execute, Morph) -> (then, Output_Get:Cosmetic Part Morph Adjustment Data Structure)
   > Set MorphMaterial()
   > Set BodyCP()
   n12 = Get EUW_CosmeticPicker_ID_Character!(self:EUW Cosmetic Pickerv 2 Object Reference=Get EUW_CosmeticPickerv2() -> (EUW_CosmeticPickerv2:EUW Cosmetic Pickerv 2 Object Reference)) -> (EUW_CosmeticPicker_ID_Character:EUW Cosmetic Pickerv 2 ID Object Reference)
   > n13 = Set SockCharacter(SockCharacter:Athena Character Item Definition Object Reference=Get character!(self:EUW Cosmetic Pickerv 2 ID Object Reference=@n12) -> (character:Athena Character Item Definition Object Reference)) -> (Output_Get:Athena Character Item Definition Object Reference)
   Event PreConstruct()
   > SetObject(self=Get DetailsView_56() -> (DetailsView_56:Details View Object Reference), NewObject=Self-Reference())
   Event Construct()
   > Bind Event to On Property Changed(
    self:Property View Base Object Reference=Get DetailsView_ID!(self:EUW Cosmetic Pickerv 2 ID Object Reference=@n12) -> (DetailsView_ID:Details View Object Reference)
    Delegate=@n11
   ) {is=title K2Node_AddDelegate}
   Event Tick()
  ]}
 GatherVariantsAndMorph {nodes [
   n1 = GatherVariantsAndMorph!() -> (then, SockCharacter:Athena Character Item Definition Object Reference) {is=title K2Node_FunctionEntry}
   > n2 = Execute Python Script!(
    execute
    PythonScript="import unreal\r\nitem_variant_name = []\r\nvariant_state = []\r\nreportmessages = []\r\n\r\nitem_variants = cid.get_editor_property(\"itemVariants\")\r\nfor entry in item_variants:\r\n\tname = str(entry.get_editor_property(\"variantChannelName\"))\r\n\titem_variant_name.append(name)\r\n\r\n\tvariantType = type(entry)\r\n\r\n\t##check if covered types have a variant\r\n\tif variantType == unreal.FortCosmeticRichColorVariant:\r\n\t\tdata = entry.get_editor_property(\"CosmeticProperties\")\r\n\t\tif len(data) == 0:\r\n\t\t\tmessage = str(name) + \" (FortCosmeticRichColorVariant) - Empty\"\r\n\t\t\treportmessages.append(message)\r\n\t\telse:\r\n\t\t\treportmessages.append(str(name) + \" (FortCosmeticRichColorVariant) - Found\")\r\n\t\t\t\r\n\telif variantType == unreal.FortCosmeticCharacterPartVariant:\r\n\t\tsize = len(entry.get_editor_property(\"PartOptions\"))\r\n\t\tfoundcount = 0\r\n\t\tfor item in entry.get_editor_property(\"PartOptions\"):\r\n\t\t\tif len(item.get_editor_property(\"CosmeticProperties\"))==0:\r\n\t\t\t\tpass\r\n\t\t\telse:\r\n\t\t\t\tfoundcount += 1\r\n\t\tmessage = str(name) + \" (FortCosmeticCharacterPartVariant) - \" + str(foundcount) + \":\"+ str(size)+\" Found\"\r\n\t\treportmessages.append(message)\r\n\t\t\t\r\n\telif variantType == unreal.FortCosmeticMaterialVariant:\r\n\t\tsize = len(entry.get_editor_property(\"MaterialOptions\"))\r\n\t\tfoundcount = 0\r\n\t\tfor item in entry.get_editor_property(\"MaterialOptions\"):\r\n\t\t\tif len(item.get_editor_property(\"CosmeticProperties\"))==0:\r\n\t\t\t\tpass\r\n\t\t\telse:\r\n\t\t\t\tfoundcount += 1\r\n\t\tmessage = str(name) + \" (FortCosmeticCharacterPartVariant) - \" + str(foundcount) + \":\"+ str(size)+\" Found\"\r\n\t\treportmessages.append(message)\r\n\telse:\r\n\t\treportmessages.append(name + \" (\" + str(variantType) + \") - Type Unsupported for check\")\r\n\r\n"
    cid=@n1
   ) -> (then, ReturnValue, reportmessages:Array of Strings) {inputs [cid], outputs [reportmessages]}
   > n3 = Set Messages(Messages:Array of Strings=@n2.reportmessages) -> (Output_Get:Array of Strings)
   > n4 = For Each Loop(Array=@n3)
   Return Node!(execute=@n4.Completed, MorphType:String=Get CompiledMessage() -> (CompiledMessage:String))
   > Set CompiledMessage(CompiledMessage:String=Append!(A=Get CompiledMessage(), B="\r\n", C:String=@n4) -> (ReturnValue)) -> (Output_Get:String)
  ]}
 Update Message {nodes [
   n1 = Update Message!() -> (then, Message:String, MessageColor:Linear Color Structure) {is=title K2Node_FunctionEntry}
   > Set Text(Text="To Text (String)"(InString=@n1), self=Get Message() -> (Message:Text Object Reference))
   > SetBrushColor(self=Get Border_0() -> (Border_0:Border Object Reference), InBrushColor=@n1.MessageColor)
  ]}
]
variables [
 BodyCP {type=CustomCharacterPart, default=null}
 InitialReportMessage {type=CosmeticPartMorphAdjustmentData, default {morphAdjustmentDataAsset=null, morphMaterialParameters=null}}
 Morph {type=CosmeticPartMorphAdjustmentData, default {morphAdjustmentDataAsset=null, morphMaterialParameters=null}}
 MorphMaterial {type=CosmeticMorphMaterialParams, default=null}
 NewVar {type=CustomCharacterPart, default=null}
 SockCharacter {type=AthenaCharacterItemDefinition, default=null}
]
widgets {
 info {rootWidgetClass=&Overlay, widgetCount=8}
 widgets [
  Overlay_0 {widgetClassPath=&Overlay, widget=Overlay_0}
  VerticalBox_51 {
   widgetClassPath=&VerticalBox
   parent=Overlay_0
   widget=VerticalBox_51
   slot {class=OverlaySlot, properties {horizontalAlignment=HAlign_Fill, verticalAlignment=VAlign_Fill}}
  }
  HorizontalBox_0 {
   widgetClassPath=&HorizontalBox
   parent=VerticalBox_51
   widget=HorizontalBox_0
   slot {class=VerticalBoxSlot, properties {size {sizeRule=Fill, value=1}}}
  }
  DetailsView_56 {
   widgetClassPath=&DetailsView
   parent=HorizontalBox_0
   widget=DetailsView_56
   properties {categoriesToShow [NoCategory]}
   slot {class=HorizontalBoxSlot}
  }
  EditorUtilityScrollBox_0 {
   widgetClassPath=&EditorUtilityScrollBox
   parent=HorizontalBox_0
   widget=EditorUtilityScrollBox_0
   properties {alwaysShowScrollbar=true}
   slot {class=HorizontalBoxSlot, properties {size {sizeRule=Fill, value=1}}}
  }
  Border_0 {
   widgetClassPath=&Border
   bIsVariable=true
   parent=EditorUtilityScrollBox_0
   widget=Border_0
   properties {brushColor=<rgba 0 0 0 1>}
   slot {class=ScrollBoxSlot, properties {size {sizeRule=Fill, value=1}}}
  }
  Message {
   widgetClassPath=&TextBlock
   bIsVariable=true
   parent=Border_0
   widget=Message
   properties {text=SelectCID}
   slot {class=BorderSlot}
  }
  EUW_CosmeticPickerv2 {
   widgetClassPath=/Game/NeverCook/ArtTools/Tech_Art/CosmeticPicker/EUW_CosmeticPickerv2
   bIsVariable=true
   parent=VerticalBox_51
   widget=EUW_CosmeticPickerv2
   slot {class=VerticalBoxSlot, properties {size {sizeRule=Fill, value=2}}}
  }
 ]
}
